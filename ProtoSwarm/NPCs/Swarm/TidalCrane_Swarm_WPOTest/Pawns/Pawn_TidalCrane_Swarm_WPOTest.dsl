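; Prototype swarm AI pawn, carrying Deimos fiend logic, that tests vertex-shader (WPO) animation on a static mesh.
; A repeating timer picks animation states from horizontal velocity and the attacking flag.
; State changes write time and alternating A/B values to custom primitive data for material blending; stance uses separate slots.
; It also handles spawn audio/FX, ice and impact cues, death erosion, deep-water despawn and shield-on-kill effects.
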(event GameplayCue|GameplayCue.Athena.InsectSwarm.BeingAttacked
  (params (EventType :type EGameplayCueEvent%20Enum :default "OnActive") (Parameters :type Gameplay%20Cue%20Parameters%20Structure%20%28by%20ref%29)))

(event GameplayCue|GameplayCue.Athena.InsectSwarm.Attacking
  (params (EventType :type EGameplayCueEvent%20Enum :default "OnActive") (Parameters :type Gameplay%20Cue%20Parameters%20Structure%20%28by%20ref%29)))

(event Custom|GetVelocityWPO
  (bind (_returnvalue_x _returnvalue_y) (Math|Vector|BreakVector (Transformation|GetVelocity)))
  (bind _returnvalue (Math|Float|Absolute(Float) _returnvalue_x))
  (bind _returnvalue_2 (Math|Float|Absolute(Float) _returnvalue_y))
  (bind _returnvalue_3 (+ _returnvalue _returnvalue_2))
  ; Prototyping WPO ANimaiton
  ;; Melee Attack
  (if (Variables|FortAIPawn|GetbAttacking)
    (WPOAnim|WPOUpdateAnim "NewEnumerator4")
    (Utilities|FlowControl|Delay 0.5)
    (Variables|FortAIPawn|SetbAttacking false))
  ;; Walk Cycle
  (if (> _returnvalue_3 6.5e+02)
    (WPOAnim|WPOUpdateAnim "NewEnumerator6")
    (elif (> _returnvalue_3 0.1)
      (WPOAnim|WPOUpdateAnim "NewEnumerator1")
      (else
        (WPOAnim|WPOUpdateAnim "NewEnumerator0")))))

(event Custom|ForceUpdateCustomDepth
  (bind _fortpawncomponent_customdepth (Variables|Default|GetFortPawnComponent_CustomDepth))
  (if (CustomDepth|IsRenderCustomDepthEnabled _fortpawncomponent_customdepth)
    (CustomDepth|Update _fortpawncomponent_customdepth 0.0 true)))

(event EventBeginPlay
  (bind _elemental_layer_audio_loop (Variables|Sound|GetElemental_Layer_Audio_Loop))
  ;[1 1 1] Begin Play
  (Utilities|FlowControl|Sequence
    (:then_0
      (Networking|SwitchHasAuthority
        (:Authority
          (EventDispatchers|BindEventtoOnDied (EventDispatchers|CreateEvent :!event Custom|CustomOnDied))
          (GameplayEffects|ApplyGameplayEffectToTarget :self (Variables|GameplayAbilities|GetAIPawnAbilitySystemComponent) :GameplayEffectClass "/Game/Athena/Deimos/Abilities/NPC/Generic/GE_HideOnMiniMap.GE_HideOnMiniMap_C" :Target (Variables|GameplayAbilities|GetAIPawnAbilitySystemComponent) :Level 0.0))
        (:Remote
          (CallFunction|ForceUpdateCustomDepth)
          (Utilities|Time|SetTimerbyEvent (delegate Custom|UpdateCustomDepth) (Variables|Default|GetCustom%20Depth%20Update%20Rate) true)
          (bind _asprimitive_component (Utilities|Casting|CastToPrimitiveComponent (Variables|/Script/Engine.MovementComponent|GetUpdatedComponent (Variables|Character|GetCharacterMovement))))
          (Variables|/Script/Engine.PrimitiveComponent|SetbGenerateOverlapEvents false _asprimitive_component)
          (CallFunction|SetupMIDs))))
    (:then_1
      (Networking|CanExecuteCosmeticEvents
        (:True
          (Audio|Components|Audio|SetSound _elemental_layer_audio_loop (Variables|Sound|GetElemental_Audio_Layer_Sound))
          (Audio|Components|Audio|Play _elemental_layer_audio_loop 0.0)
          (bind _returnvalue (Utilities|Time|SetTimerbyFunctionName 0 "CharacterSpawnInSafetyCheck" 5.5))
          (Variables|Effects|SetCharacterSpawnInSafetyCheckHandle _returnvalue)
          (SoundLibrary|AddLibrariesfromCollections (WorldSubsystems|GetSoundLibrarySubsystem) self (Variables|Default|GetSound%20Library%20Collection%20Tags))
          (CallFunction|PlaySoundLibraryAudio "(TagName=\"SoundLibrary.Deimos.Spawn\")")
          (Utilities|FlowControl|Sequence
            (:then_0
              ;[1 1 1] Play spawn fx if cheat activated
              (bind _returnvalue_2 (Development|GetConsoleVariableBoolValue "AI.Deimos.Debug_SpawnWithSidewaysFX"))
              (if _returnvalue_2
                (GameplayCue|ExecuteGameplayCue%28Burst%29Local self (Variables|Default|GetGC_SidewaysSpawnFx) (Ability|GameplayCue|MakeGameplayCueParameters :NormalizedMagnitude 0.0 :RawMagnitude 0.0 :Location (Transformation|GetActorLocation) :Normal "0,0,0" :GameplayEffectLevel 1 :AbilityLevel 1 :bReplicateLocationWhenUsingMinimalRepProxy false))))
            (:then_1
              ;[1 1 1] Play spawn fx if it is set
              (if (GameplayTags|NotEqual(GameplayTag) :A (Variables|Effects|GetGC_OnSpawnFX))
                (Networking|CanExecuteCosmeticEvents
                  (:True
                    (GameplayCue|ExecuteGameplayCue%28Burst%29Local self (Variables|Effects|GetGC_OnSpawnFX) (Ability|GameplayCue|MakeGameplayCueParameters :NormalizedMagnitude 0.0 :RawMagnitude 0.0 :Location (Transformation|GetActorLocation) :Normal "0,0,0" :GameplayEffectLevel 1 :AbilityLevel 1 :bReplicateLocationWhenUsingMinimalRepProxy false))))))
            (:then_2
              (Niagara|SpawnSystematLocation (Variables|Default|GetSpawnInVFX) (Transformation|GetActorLocation) (Transformation|GetActorRotation) "1.000000,1.000000,1.000000" true true "None" false))))))
    (:then_2
      (Utilities|Time|SetTimerbyEvent (delegate Custom|GetVelocityWPO) 0.15 true))))

(event Custom|UpdateCustomDepth
  (bind _fortpawncomponent_customdepth (Variables|Default|GetFortPawnComponent_CustomDepth))
  ;[1 1 1] Begin Play
  (if (CustomDepth|IsRenderCustomDepthEnabled _fortpawncomponent_customdepth)
    (CustomDepth|Update _fortpawncomponent_customdepth (Variables|Default|GetCustom%20Depth%20Update%20Rate) false)))

(event WaterBody|EventOnWaterDeepNavMeshEnter
  ;[1 1 1] Kill Deimos NPC when entering Deep Water (using navmesh because physics overlaps might be optimized out)
  (Networking|SwitchHasAuthority
    (:Authority
      (if (GameplayTags|HasMatchingGameplayTag self "(TagName=\"Pawn.Athena.NPC.Deimos\")")
        (if (not (Variables|Health|GetbIsDying))
          (Utilities|FlowControl|Delay 1.0)
          (if (not (Variables|Health|GetbIsDying))
            (FortPawn|ForceKill self "(TagName=\"Gameplay.Effect.InstantDeath.Despawn\")" (Pawn|GetController) self) ; If we are Deimos and hit deep water, despawn.\n\nOther Children of this pawn override this Event and do their own thing. \nSome drown, some do not.
          ))))))

(event WaterBody|EventOnBeginDeepWaterInteraction
  (params (WaterSurfaceLocation :type Vector :default "0, 0, 0") (WaterDepth :type Float%20%28single-precision%29) (WaterBody :type Water%20Body%20Object%20Reference))
  ;[1 1 1] Kill Deimos NPC when entering Deep Water (using navmesh because physics overlaps might be optimized out)
  (Networking|SwitchHasAuthority
    (:Authority
      (if (GameplayTags|HasMatchingGameplayTag self "(TagName=\"Pawn.Athena.NPC.Deimos\")")
        (if (not (Variables|Health|GetbIsDying))
          (FortPawn|ForceKill self "(TagName=\"Gameplay.Effect.InstantDeath.Despawn\")" (Pawn|GetController) self) ; If we are Deimos and hit deep water, despawn.\n\nOther Children of this pawn override this Event and do their own thing. \nSome drown, some do not.
        )))))

(event EventOnDeathServer
  (params (Damage :type Float%20%28single-precision%29) (DamageTags :type Gameplay%20Tag%20Container%20Structure%20%28by%20ref%29) (Momentum :type Vector :default "0, 0, 0") (HitInfo :type Hit%20Result%20Structure%20%28by%20ref%29) (InstigatedBy :type Controller%20Object%20Reference) (DamageCauser :type Actor%20Object%20Reference) (EffectContext :type Gameplay%20Effect%20Context%20Handle%20Structure))
  (|CallParentFunction Damage DamageTags Momentum HitInfo InstigatedBy DamageCauser EffectContext)
  (CallFunction|SpawnActoronDeath))

(event GameplayCue|GameplayCue.Athena.IceDeimos
  (params (EventType :type EGameplayCueEvent%20Enum :default "OnActive") (Parameters :type Gameplay%20Cue%20Parameters%20Structure%20%28by%20ref%29))
  ;[0 0.85 1] Application and removal of Chiller Ice material
  (switch Utilities|FlowControl|Switch|SwitchonEGameplayCueEvent EventType
    (:OnActive
      (Rendering|Material|SetScalarParameterValueonMaterials (Variables|Default|GetStaticMesh) "Ice" 1.0)
      (Utilities|Time|SetTimerbyEvent (delegate Custom|RemoveChillerMaterial) (ScalableFloat|GetScalableFloatMultiplier (Variables|Default|GetFreezeDuration))))))

(event Custom|RemoveChillerMaterial
  ;[0 0.85 1] Application and removal of Chiller Ice material
  (Rendering|Material|SetScalarParameterValueonMaterials (Variables|Default|GetStaticMesh) "Ice" 0.0))

(event Custom|CustomOnDied
  (params (DamagedActor :type Actor%20Object%20Reference) (Damage :type Float%20%28single-precision%29) (InstigatedBy :type Controller%20Object%20Reference) (DamageCauser :type Actor%20Object%20Reference) (HitLocation :type Vector) (FHitComponent :type Primitive%20Component%20Object%20Reference) (BoneName :type Name) (Momentum :type Vector))
  ;[1 1 1] Death material erosion
  (CallFunction|AnimateDeathMaterial)
  (CallFunction|CEHideBeamFXonDeimos)
  ;;[0.3509 0.6383 1] FORT-684077 | Add additional check to determine if rewarding shield in FNE modes
  (Networking|SwitchHasAuthority
    (:Authority
      (if (Fortnite|IsSpatialGameplayExperience)
        (if (GameplayTags|HasMatchingGameplayTag self "(TagName=\"Pawn.Athena.NPC.Deimos.DisableShieldOnElimination\")")
          _
          (else
            _))
        (else
          _))
      (bind _asfort_player_controller_athena (Utilities|Casting|CastToFortPlayerControllerAthena InstigatedBy))
      (Utilities|IsValid (Ability|GetAbilitySystemComponent (Game|Player|GetPlayerPawn _asfort_player_controller_athena))
        (:Is%20Valid
          (if (not (Health|IsDBNO (Game|Player|GetPlayerPawn _asfort_player_controller_athena)))
            (GameplayEffects|ApplyGameplayEffectToSelf :self (Ability|GetAbilitySystemComponent (Game|Player|GetPlayerPawn _asfort_player_controller_athena)) :GameplayEffectClass (Variables|Default|GetGE_DeimosShieldOnFiendKill) :Level 0.0)))))))

(event Custom|AnimateDeathMaterial
  (bind _mid_maskeddeathmaterial (Variables|Default|GetMID_MaskedDeathMaterial))
  ;[1 1 1] Death material erosion
  (Networking|SwitchHasAuthority
    (:Remote
      (Rendering|Material|SetMaterial (Variables|Default|GetStaticMesh) 0 _mid_maskeddeathmaterial)
      (Rendering|Material|SetScalarParameterValue _mid_maskeddeathmaterial "Death_CurrentTime" (Utilities|Time|GetGameTimeinSeconds) :!class /Script/Engine.MaterialInstanceDynamic))))

(event GameplayCue|GameplayCue.Impact.Physical.Creature
  (params (EventType :type EGameplayCueEvent%20Enum :default "OnActive") (Parameters :type Gameplay%20Cue%20Parameters%20Structure%20%28by%20ref%29))
  (bind (_bblockinghit _binitialoverlap _time _distance _location) (Collision|BreakHitResult (Ability|GameplayCue|GetHitResult Parameters)))
  ;[1 1 1] ImpactATarget
  (Audio|PlaySoundatLocation (Sound|GetImpactPhysicalSurfaceSound (Ability|GameplayCue|GetHitResult Parameters)) _location)
  (Default|PlayDamageFXSurfaceEnemies Parameters (Variables|Effects|GetImpactParticles_Default) (Variables|Effects|GetImpactParticles_Wood) (Variables|Effects|GetImpactParticles_Stone) (Variables|Effects|GetImpactParticles_Metal) (Variables|Effects|GetImpactParticles_HumanFlesh) (.location (Transformation|GetSocketTransform (Variables|Default|GetStaticMesh) (Variables|Default|GetImpact_MeleeSocket) "RTS_World")) (.rotation (Transformation|GetSocketTransform (Variables|Default|GetStaticMesh) (Variables|Default|GetImpact_MeleeSocket) "RTS_World")) (Variables|Default|GetMeleeImpact_UseSocketTransforms)))

(event Custom|CE_HideBeamFXonDeimos)

(fn ApplyAwakenEffectToSkeletalMesh
  (params (Mesh :type Skeletal%20Mesh%20Component%20Object%20Reference))
  (bind _awakenfxskeletalmeshes (Variables|Effects|GetAwakenFXSkeletalMeshes))
  (bind (_value _returnvalue) (Utilities|Map|Find _awakenfxskeletalmeshes Mesh))
  (bind _fxskeletalmesh (LocalVariables|GetFXSkeletalMesh))
  (Utilities|IsValid Mesh
    (:Is%20Valid
      (if _returnvalue
        (bind _output_get (LocalVariables|SetFXSkeletalMesh _value))
        (Utilities|IsValid _output_get
          (:Is%20Valid
            _)
          (:Is%20Not%20Valid
            (bind _returnvalue_2 (AddComponent|NODE_AddFXSkeletonMeshComponent-6))
            (Utilities|Map|Add (Variables|Effects|GetAwakenFXSkeletalMeshes) Mesh _returnvalue_2)
            (bind _output_get_2 (LocalVariables|SetFXSkeletalMesh _returnvalue_2))
            (Class|FortFXSkeletonMeshComponent|SetSource _output_get_2 Mesh)
            (Variables|/Script/FortniteGame.FortFXSkeletonMeshComponent|SetAwakenDuration (Variables|Effects|GetAwakenDuration) _fxskeletalmesh)))
        (Class|FXSkeletonMeshComponent|ApplyAwakenEffect _fxskeletalmesh)
        (else
          _)))))

(fn ApplyAwakenEffectToStaticMesh
  (params (Mesh :type Static%20Mesh%20Component%20Object%20Reference))
  (bind _awakenfxstaticmeshes (Variables|Effects|GetAwakenFXStaticMeshes))
  (bind (_value _returnvalue) (Utilities|Map|Find _awakenfxstaticmeshes Mesh))
  (bind _fxstaticmesh (LocalVariables|GetFXStaticMesh))
  (Utilities|IsValid Mesh
    (:Is%20Valid
      (if _returnvalue
        (bind _output_get (LocalVariables|SetFXStaticMesh _value))
        (Utilities|IsValid _output_get
          (:Is%20Valid
            _)
          (:Is%20Not%20Valid
            (bind _returnvalue_2 (AddComponent|NODE_AddFXStaticMeshComponent-2))
            (Utilities|Map|Add (Variables|Effects|GetAwakenFXStaticMeshes) Mesh _returnvalue_2)
            (bind _output_get_2 (LocalVariables|SetFXStaticMesh _returnvalue_2))
            (Class|FortFXStaticMeshComponent|SetSource _output_get_2 Mesh)
            (Variables|/Script/FortniteGame.FortFXStaticMeshComponent|SetAwakenDuration (Variables|Effects|GetAwakenDuration) _fxstaticmesh)))
        (Class|FXStaticMeshComponent|ApplyAwakenEffect _fxstaticmesh)
        (else
          _)))))

(fn ApplyBuildingHitEffectToSkeletalMesh
  (params (Mesh :type Skeletal%20Mesh%20Component%20Object%20Reference))
  (bind _buildinghitfxskeletalmeshes (Variables|Effects|GetBuildingHitFXSkeletalMeshes))
  (bind (_value _returnvalue) (Utilities|Map|Find _buildinghitfxskeletalmeshes Mesh))
  (bind _fxskeletalmesh (LocalVariables|GetFXSkeletalMesh))
  (Utilities|IsValid Mesh
    (:Is%20Valid
      (if _returnvalue
        (bind _output_get (LocalVariables|SetFXSkeletalMesh _value))
        (Utilities|IsValid _output_get
          (:Is%20Valid
            _)
          (:Is%20Not%20Valid
            (bind _returnvalue_2 (AddComponent|NODE_AddFXSkeletonMeshComponent-2))
            (Utilities|Map|Add (Variables|Effects|GetBuildingHitFXSkeletalMeshes) Mesh _returnvalue_2)
            (bind _output_get_2 (LocalVariables|SetFXSkeletalMesh _returnvalue_2))
            (Class|FortFXSkeletonMeshComponent|SetSource _output_get_2 Mesh)
            (Variables|/Script/FortniteGame.FortFXSkeletonMeshComponent|SetBuildingHitDuration (Variables|Effects|GetBuildingHitDuration) _fxskeletalmesh)))
        (Class|FXSkeletonMeshComponent|ApplyBuildingHitEffect _fxskeletalmesh)
        (else
          _)))))

(fn ApplyBuildingHitEffectToStaticMesh
  (params (Mesh :type Static%20Mesh%20Component%20Object%20Reference))
  (bind _buildinghitfxstaticmeshes (Variables|Effects|GetBuildingHitFXStaticMeshes))
  (bind (_value _returnvalue) (Utilities|Map|Find _buildinghitfxstaticmeshes Mesh))
  (bind _fxstaticmesh (LocalVariables|GetFXStaticMesh))
  (Utilities|IsValid Mesh
    (:Is%20Valid
      (if _returnvalue
        (bind _output_get (LocalVariables|SetFXStaticMesh _value))
        (Utilities|IsValid _output_get
          (:Is%20Valid
            _)
          (:Is%20Not%20Valid
            (bind _returnvalue_2 (AddComponent|NODE_AddFXStaticMeshComponent-0))
            (Utilities|Map|Add (Variables|Effects|GetBuildingHitFXStaticMeshes) Mesh _returnvalue_2)
            (bind _output_get_2 (LocalVariables|SetFXStaticMesh _returnvalue_2))
            (Class|FortFXStaticMeshComponent|SetSource _output_get_2 Mesh)
            (Variables|/Script/FortniteGame.FortFXStaticMeshComponent|SetBuildingHitDuration (Variables|Effects|GetBuildingHitDuration) _fxstaticmesh)))
        (Class|FXStaticMeshComponent|ApplyBuildingHitEffect _fxstaticmesh)
        (else
          _)))))

(fn CharacterSpawnInSafetyCheck
  (if (not (Variables|Effects|GetSpawnInTimelineCompletedSuccessfully))
    (Components|Timeline|Stop (Variables|Pawn_TidalCrane_Swarm_WPOTest|GetEnemySpawnInTL))
    (CallFunction|SpawninSafetyCheck)))

(fn DeregisterEffectSkeletalMesh
  (params (Mesh :type Skeletal%20Mesh%20Component%20Object%20Reference))
  (bind _awakenfxskeletalmeshes (Variables|Effects|GetAwakenFXSkeletalMeshes))
  (bind (_value _returnvalue) (Utilities|Map|Find _awakenfxskeletalmeshes Mesh))
  (bind _buildinghitfxskeletalmeshes (Variables|Effects|GetBuildingHitFXSkeletalMeshes))
  (bind (_value_2 _returnvalue_2) (Utilities|Map|Find _buildinghitfxskeletalmeshes Mesh))
  (Utilities|IsValid Mesh
    (:Is%20Valid
      (if _returnvalue
        (Utilities|IsValid _value
          (:Is%20Valid
            (Components|DestroyComponent _value))
          (:Is%20Not%20Valid
            _))
        (Utilities|Map|Remove (Variables|Effects|GetAwakenFXSkeletalMeshes) Mesh)
        (if _returnvalue_2
          (Utilities|IsValid _value_2
            (:Is%20Valid
              (Components|DestroyComponent _value_2))
            (:Is%20Not%20Valid
              _))
          (Utilities|Map|Remove (Variables|Effects|GetBuildingHitFXSkeletalMeshes) Mesh)
          (Utilities|Array|RemoveItem (Variables|Effects|GetRegisteredSkeletalMeshesForEffects) Mesh))))))

(fn DeregisterEffectStaticMesh
  (params (Mesh :type Static%20Mesh%20Component%20Object%20Reference))
  (bind _awakenfxstaticmeshes (Variables|Effects|GetAwakenFXStaticMeshes))
  (bind (_value _returnvalue) (Utilities|Map|Find _awakenfxstaticmeshes Mesh))
  (bind _buildinghitfxstaticmeshes (Variables|Effects|GetBuildingHitFXStaticMeshes))
  (bind (_value_2 _returnvalue_2) (Utilities|Map|Find _buildinghitfxstaticmeshes Mesh))
  (Utilities|IsValid Mesh
    (:Is%20Valid
      (if _returnvalue
        (Utilities|IsValid _value
          (:Is%20Valid
            (Components|DestroyComponent _value))
          (:Is%20Not%20Valid
            _))
        (Utilities|Map|Remove (Variables|Effects|GetAwakenFXStaticMeshes) Mesh)
        (if _returnvalue_2
          (Utilities|IsValid _value_2
            (:Is%20Valid
              (Components|DestroyComponent _value_2))
            (:Is%20Not%20Valid
              _))
          (Utilities|Map|Remove (Variables|Effects|GetBuildingHitFXStaticMeshes) Mesh)
          (Utilities|Array|RemoveItem (Variables|Effects|GetRegisteredStaticMeshesForEffects) Mesh))))))

(fn OnRep_Anim_NextState)

(fn OnRep_Anim_StateA_Enum)

(fn Orphaned
  (return false 0))

(fn PlayAdditiveHitReacts
  (params (Hit%20Direction :type Vector) (Anim%20Montage :type Anim%20Montage%20Object%20Reference))
  (HitReacts|PlayAdditiveHitReactsNative Hit%20Direction Anim%20Montage) ; This used to be done directly in the Event graph, which meant any Child of this Pawn couldn't override the Function behavior.\n\nWith this, we should be able to override PlayAdditiveHitReacts and call the Native function here with different logic in front of it, if we need to.
)

(fn PlaySoundLibraryAudio
  (params (Event%20Name :type Gameplay%20Tag%20Structure))
  (Networking|CanExecuteCosmeticEvents
    (:True
      (if (Utilities|IsValid (Variables|Sound|GetSoundLibraryContext) :!class /Script/Engine.KismetSystemLibrary)
        (Utilities|Struct|SetmembersinSimpleSoundSettings (Variables|/Script/SoundLibrary.SoundLibrarySimpleContext|GetSettings (Variables|Sound|GetSoundLibraryContext)) Event%20Name)
        (else
          (bind _returnvalue (SoundLibrary|CreateSoundLibraryContext :Class "/Script/SoundLibrary.SoundLibrarySimpleContext" :Settings (Utilities|Struct|MakeSimpleSoundSettings Event%20Name) :Actor self :bForwardToActorOwner true :bForwardToAttachedActors true :bAutoResetContext true))
          (bind _output_get (Variables|Sound|SetSoundLibraryContext _returnvalue))))
      (SoundLibrary|PlaySoundwithContext (WorldSubsystems|GetSoundLibrarySubsystem) _output_get))))

(fn RegisterEffectSkeletalMesh
  (params (Mesh :type Skeletal%20Mesh%20Component%20Object%20Reference))
  (Utilities|IsValid Mesh
    (:Is%20Valid
      (Utilities|Array|Add (Variables|Effects|GetRegisteredSkeletalMeshesForEffects) Mesh))
    (:Is%20Not%20Valid
      (Development|PrintString "Attempting to register invalid skeletal mesh for husk building hit FX"))))

(fn RegisterEffectStaticMesh
  (params (Mesh :type Static%20Mesh%20Component%20Object%20Reference))
  (Utilities|IsValid Mesh
    (:Is%20Valid
      (Utilities|Array|Add (Variables|Effects|GetRegisteredStaticMeshesForEffects) Mesh))
    (:Is%20Not%20Valid
      (Development|PrintString "Attempting to register invalid static mesh for husk building hit FX"))))

(fn SetScalarParameterOnAllCharacterMIDs
  (params (Parameter%20Name :type Name) (Scalar%20Value :type Float%20%28double-precision%29) (AlsoModifyOriginalMIDs :type Boolean))
  (bind _auxillary_objects (Variables|Effects|GetAuxillary%20Objects))
  (bind _duplicate_auxillary_objects (Variables|Effects|GetDuplicate%20Auxillary%20Objects))
  (Utilities|FlowControl|Sequence
    (:then_0
      (bind _mesh (Variables|Character|GetMesh)
        (:then
          (Rendering|Material|SetScalarParameterValueonMaterials _mesh (Variables|GetParameter%20Name) (Variables|GetScalar%20Value)))
        (:else
          _))
      (bind _duplicatecharactermesh (Variables|Effects|GetDuplicateCharacterMesh)
        (:then
          (Rendering|Material|SetScalarParameterValueonMaterials _duplicatecharactermesh (Variables|GetParameter%20Name) (Variables|GetScalar%20Value)))
        (:else)))
    (:then_1
      (if (Utilities|IsArrayNonEmpty _auxillary_objects)
        (for _array_element _auxillary_objects
          (Utilities|IsValid _array_element
            (:Is%20Valid
              (bind _asmesh_component (Utilities|Casting|CastToMeshComponent _array_element))
              (Rendering|Material|SetScalarParameterValueonMaterials _asmesh_component (Variables|GetParameter%20Name) (Variables|GetScalar%20Value)))))))
    (:then_2
      (if (Utilities|IsArrayNonEmpty _duplicate_auxillary_objects)
        (for _array_element_2 _duplicate_auxillary_objects
          (Utilities|IsValid _array_element_2
            (:Is%20Valid
              (bind _asmesh_component_2 (Utilities|Casting|CastToMeshComponent _array_element_2))
              (Rendering|Material|SetScalarParameterValueonMaterials _asmesh_component_2 (Variables|GetParameter%20Name) (Variables|GetScalar%20Value)))))))
    (:then_3
      ;[1 1 1] Optionally keep the original MIDs in sync with the override MIDs
      (if AlsoModifyOriginalMIDs
        (Utilities|FlowControl|Sequence
          (:then_0
            (for _array_element_3 (Variables|Materials|GetPrevious%20MID)
              (Rendering|Material|SetScalarParameterValue _array_element_3 (Variables|GetParameter%20Name) (Variables|GetScalar%20Value) :!class /Script/Engine.MaterialInstanceDynamic)))
          (:then_1
            (for _array_element_4 (Variables|Materials|GetPrevious%20MID_AuxObjs)
              (Rendering|Material|SetScalarParameterValue _array_element_4 (Variables|GetParameter%20Name) (Variables|GetScalar%20Value) :!class /Script/Engine.MaterialInstanceDynamic))))))))

(fn SetupMIDs
  (bind _staticmesh (Variables|Default|GetStaticMesh))
  (bind _returnvalue (Rendering|Material|CreateDynamicMaterialInstance (Variables|Default|GetMaskedDeathMaterial) "None" "None" :!class /Script/Engine.KismetMaterialLibrary))
  (Variables|Default|SetMID_MaskedDeathMaterial _returnvalue)
  (bind _returnvalue_2 (Rendering|Material|CreateDynamicMaterialInstance (Rendering|Material|GetMaterial _staticmesh 0) "None" "None" :!class /Script/Engine.KismetMaterialLibrary))
  (Variables|Default|SetMID_MonsterMaterial _returnvalue_2)
  (Rendering|Material|SetMaterial _staticmesh 0 _returnvalue_2))

(fn SetVectorParameterOnAllCharacterMIDs
  (params (Parameter%20Name :type Name) (Vector%20Value :type Vector) (Propagate%20to%20Auxiliary%20Meshes :type Boolean :default true))
  (bind _auxillary_objects (Variables|Effects|GetAuxillary%20Objects))
  (bind _duplicate_auxillary_objects (Variables|Effects|GetDuplicate%20Auxillary%20Objects))
  (bind _mesh (Variables|Character|GetMesh)
    (:then
      (Rendering|Material|SetVectorParameterValueonMaterials _mesh Parameter%20Name Vector%20Value))
    (:else
      _))
  (bind _duplicatecharactermesh (Variables|Effects|GetDuplicateCharacterMesh)
    (:then
      (Rendering|Material|SetVectorParameterValueonMaterials _duplicatecharactermesh Parameter%20Name Vector%20Value))
    (:else
      _))
  (if (and (> (Utilities|Array|Length _auxillary_objects) 0) (Variables|GetPropagate%20to%20Auxiliary%20Meshes))
    (for _array_element _auxillary_objects
      (Utilities|IsValid _array_element
        (:Is%20Valid
          (bind _asmesh_component (Utilities|Casting|CastToMeshComponent _array_element))
          (Rendering|Material|SetVectorParameterValueonMaterials _asmesh_component (Variables|GetParameter%20Name) (Variables|GetVector%20Value)))))
    (if (and (> (Utilities|Array|Length _duplicate_auxillary_objects) 0) (Variables|GetPropagate%20to%20Auxiliary%20Meshes))
      (for _array_element_2 _duplicate_auxillary_objects
        (Utilities|IsValid _array_element_2
          (:Is%20Valid
            (bind _asmesh_component_2 (Utilities|Casting|CastToMeshComponent _array_element_2))
            (Rendering|Material|SetVectorParameterValueonMaterials _asmesh_component_2 (Variables|GetParameter%20Name) (Variables|GetVector%20Value)))))))
  (return))

(fn SpawnActorOnDeath)

(fn SpawnParticleSystemOnCharacterMesh
  (params (ParticleSystemTemplate :type Particle%20System%20Object%20Reference) (ParticleSystemComponentReferenceVar :type Cascade%20Particle%20System%20Component%20Object%20Reference) (AttachPointName :type Name) (Location :type Vector) (Rotation :type Rotator) (InstanceParameters :type Array%20of%20Particle%20Sys%20Param%20Structures) (AutoActivate :type Boolean :default true) (AutoDestroy :type Boolean) (AbsoluteLocation :type Boolean) (AbsoluteRotation :type Boolean) (AbsoluteScale :type Boolean))
  (Networking|CanExecuteCosmeticEvents
    (:True
      (Utilities|IsValid ParticleSystemTemplate
        (:Is%20Valid
          (Utilities|IsValid ParticleSystemComponentReferenceVar
            (:Is%20Valid
              (Components|Activation|Deactivate ParticleSystemComponentReferenceVar))
            (:Is%20Not%20Valid
              _))
          (bind _returnvalue (Effects|SpawnEmitterAttached ParticleSystemTemplate (Variables|Default|GetStaticMesh) AttachPointName Location Rotation "1.000000,1.000000,1.000000" "KeepRelativeOffset" AutoDestroy))
          (Variables|/Script/CascadeRuntime.ParticleSystemComponent|SetInstanceParameters InstanceParameters _returnvalue)
          (Components|Activation|SetActive _returnvalue AutoActivate false)
          (Transformation|SetAbsolute _returnvalue AbsoluteLocation AbsoluteRotation AbsoluteScale)
          (return _returnvalue))))))

(fn ConstructionScript
  (Networking|CanExecuteCosmeticEvents))

(fn WPO_CycleAllAnimations
  (params (BlendTo :type E_WPOSwarmerAnimStates%20Enum :default "NewEnumerator0")))

(fn WPO_StanceAgressive_OFF
  (Rendering|Material|SetCustomPrimitiveDataFloat (Variables|Default|GetStaticMesh) 5 0.0)
  (Rendering|Material|SetCustomPrimitiveDataFloat (Variables|Default|GetStaticMesh) 6 (Utilities|Time|GetGameTimeinSeconds)))

(fn WPO_StanceAgressive_ON
  (Rendering|Material|SetCustomPrimitiveDataFloat (Variables|Default|GetStaticMesh) 5 1.0)
  (Rendering|Material|SetCustomPrimitiveDataFloat (Variables|Default|GetStaticMesh) 6 (Utilities|Time|GetGameTimeinSeconds)))

(fn WPO_UpdateAnim
  (params (BlendTo :type E_WPOSwarmerAnimStates%20Enum :default "NewEnumerator0"))
  (if (Utilities|Enum|Equal(Enum) BlendTo (Variables|WPOAnim|GetAnim_CurrentState))
    _
    (elif (Variables|WPOAnim|GetAnim_A>B)
      (bind _output_get (Variables|WPOAnim|SetAnim_StateB_Enum BlendTo))
      (Rendering|Material|SetCustomPrimitiveDataVector4 (Variables|Default|GetStaticMesh) 10 (Math|Vector4|MakeVector4 (Utilities|Time|GetGameTimeinSeconds) (Math|Conversions|ToFloat(Boolean) (Variables|WPOAnim|GetAnim_A>B)) (Utilities|Select :Index (Variables|WPOAnim|GetAnim_StateA_Enum) :NewEnumerator0 0.5 :NewEnumerator1 1.2 :NewEnumerator6 2.2 :NewEnumerator7 3.5 :NewEnumerator4 4.2 :NewEnumerator3 5.2 :NewEnumerator5 6.5 :NewEnumerator2 7.5) (Utilities|Select :Index _output_get :NewEnumerator0 0.5 :NewEnumerator1 1.2 :NewEnumerator6 2.2 :NewEnumerator7 3.5 :NewEnumerator4 4.2 :NewEnumerator3 5.2 :NewEnumerator5 6.5 :NewEnumerator2 7.5)))
      (Variables|WPOAnim|SetAnim_A>B false)
      (bind _output_get_2 (Variables|WPOAnim|SetAnim_CurrentState BlendTo))
      (Development|PrintString (Utilities|String|EnumtoString _output_get_2) true true (Utilities|Select :Index _output_get_2 :NewEnumerator0 "(R=1.000000,G=1.000000,B=1.000000,A=1.000000)" :NewEnumerator1 "(R=0.000000,G=0.936360,B=1.000000,A=1.000000)" :NewEnumerator6 "(R=0.097165,G=1.000000,B=0.000000,A=1.000000)" :NewEnumerator4 "(R=1.000000,G=0.022955,B=0.000000,A=1.000000)" :NewEnumerator3 "(R=0.025558,G=0.000000,B=1.000000,A=1.000000)" :NewEnumerator5 "(R=1.000000,G=0.000000,B=0.866227,A=1.000000)" :NewEnumerator2 "(R=0.000000,G=0.000000,B=0.000000,A=1.000000)"))
      (else
        (bind _output_get_3 (Variables|WPOAnim|SetAnim_StateA_Enum BlendTo))
        (Rendering|Material|SetCustomPrimitiveDataVector4 (Variables|Default|GetStaticMesh) 10 (Math|Vector4|MakeVector4 (Utilities|Time|GetGameTimeinSeconds) (Math|Conversions|ToFloat(Boolean) (Variables|WPOAnim|GetAnim_A>B)) (Utilities|Select :Index _output_get_3 :NewEnumerator0 0.5 :NewEnumerator1 1.2 :NewEnumerator6 2.2 :NewEnumerator7 3.5 :NewEnumerator4 4.2 :NewEnumerator3 5.2 :NewEnumerator5 6.5 :NewEnumerator2 7.5) (Utilities|Select :Index (Variables|WPOAnim|GetAnim_StateB_Enum) :NewEnumerator0 0.5 :NewEnumerator1 1.2 :NewEnumerator6 2.2 :NewEnumerator7 3.5 :NewEnumerator4 4.2 :NewEnumerator3 5.2 :NewEnumerator5 6.5 :NewEnumerator2 7.5)))
        (Variables|WPOAnim|SetAnim_A>B true)))))

(fn WPO_UpdateStance
  (params (Stance_Agressive :type Boolean))
  (if (not (== Stance_Agressive (Variables|WPOAnim|GetStance_IsAgressive)))
    (bind _output_get (Variables|WPOAnim|SetStance_IsAgressive Stance_Agressive))
    (Rendering|Material|SetCustomPrimitiveDataFloat (Variables|Default|GetStaticMesh) 5 (Math|Conversions|ToFloat(Boolean) _output_get))
    (Rendering|Material|SetCustomPrimitiveDataFloat (Variables|Default|GetStaticMesh) 6 (Utilities|Time|GetGameTimeinSeconds))))
